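# Sound cue for the closing of a metal air duct fan door among the Apollo air duct props.
# A single wave player feeds a Modulator node, which varies the one close sample on playback.
# The cue plays at reduced volume through the Fort door sound class and door attenuation.
# A concurrency override prevents new instances from starting once the limit is reached.
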SoundCue /Game/Sounds/Apollo/Props/AirDucts/Door_Metal_AirDuct_FanClose_Cue
root=Modulator_4
nodes [
 {id=Modulator_4, children [WavePlayer_10]}
 {id=WavePlayer_10, properties {soundWaveAssetPtr=&/Door_Metal_Air_Duct_Fan_Close_01}}
]
properties {
 attenuationSettings=/Game/Sounds/Attenuation/Fort_Doors_Att
 bOverrideConcurrency=true
 maxDistance=2400
 soundClassObject=&Fort_Doors
 volumeMultiplier=0.4
 concurrencyOverrides {resolutionRule=PreventNew}
}
