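# Sound cue for bullet impacts on metal air duct props, routed to the impacts sound class with prop impact attenuation.
# A quality-level node sends its first two slots to a random pick of nine variations and its last two to a four-variation subset.
# The smaller random node is excluded from branch culling.
# An object-impact concurrency set is assigned, alongside stop-oldest override values.
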SoundCue /Game/Sounds/Apollo/Props/AirDucts/Impact_Airduct_Cue
root=QualityLevel_2
nodes [
 {id=QualityLevel_2, children [Random_0,Random_0,Random_4,Random_4]}
 {
  id=Random_0
  children [WavePlayer_5, WavePlayer_9, WavePlayer_10, WavePlayer_11, WavePlayer_12, WavePlayer_13, WavePlayer_14, WavePlayer_15, WavePlayer_16]
  properties {weights [1,1,1,1,1,1,1,1,1]}
 }
 {
  id=Random_4
  children [WavePlayer_17, WavePlayer_18, WavePlayer_19, WavePlayer_20]
  properties {bShouldExcludeFromBranchCulling=true, weights [1,1,1,1]}
 }
 {id=WavePlayer_10, properties {soundWaveAssetPtr=&/Bullet_Impact_Metal_Air_Duct_02}}
 {id=WavePlayer_11, properties {soundWaveAssetPtr=&/Bullet_Impact_Metal_Air_Duct_03}}
 {id=WavePlayer_12, properties {soundWaveAssetPtr=&/Bullet_Impact_Metal_Air_Duct_04}}
 {id=WavePlayer_13, properties {soundWaveAssetPtr=&/Bullet_Impact_Metal_Air_Duct_05}}
 {id=WavePlayer_14, properties {soundWaveAssetPtr=&/Bullet_Impact_Metal_Air_Duct_07}}
 {id=WavePlayer_15, properties {soundWaveAssetPtr=&/Bullet_Impact_Metal_Air_Duct_06}}
 {id=WavePlayer_16, properties {soundWaveAssetPtr=&/Bullet_Impact_Metal_Air_Duct_08}}
 {id=WavePlayer_17, properties {soundWaveAssetPtr=&/Bullet_Impact_Metal_Air_Duct_09}}
 {id=WavePlayer_18, properties {soundWaveAssetPtr=&/Bullet_Impact_Metal_Air_Duct_01}}
 {id=WavePlayer_19, properties {soundWaveAssetPtr=&Bullet_Impact_Metal_Air_Duct_02}}
 {id=WavePlayer_20, properties {soundWaveAssetPtr=&Bullet_Impact_Metal_Air_Duct_03}}
 {id=WavePlayer_5, properties {soundWaveAssetPtr=&Bullet_Impact_Metal_Air_Duct_09}}
 {id=WavePlayer_9, properties {soundWaveAssetPtr=&Bullet_Impact_Metal_Air_Duct_01}}
]
properties {
 attenuationSettings=/Game/Sounds/Attenuation/Fort_Prop_Impact_Att
 maxDistance=3256
 soundClassObject=&Fort_Impacts
 volumeMultiplier=0.96
 concurrencyOverrides {maxCount=3, resolutionRule=StopOldest}
 concurrencySet [&CONC_Object_Impacts_FourMax]
}
